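; Creative vehicle spawner device for the MotorBoat (MeatballVehicle_L), extending a parent spawner with boat-specific weapon setup.
; On a vehicle config update, it calls the parent, casts the cached vehicle, then applies the weapon-enabled option through DisableWeapon and the bEnableWeapons scalable float.
; UpdateEffects plays the trap-ready sound at the device location where cosmetic events can run and the device is enabled.
; HandleEditModeCollision is empty, and the construction script only calls its parent.
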(event EventUpdateVehicleConfigs
  (bind _bweaponenabled (Variables|Default|GetbWeaponEnabled))
  ;[0.8079 1 0.6756 0.75] Handle Weapons Option on Meatball Vehicle by Setting Value
  (|CallParentFunction)
  (bind _asmeatball_vehicle_l (Utilities|Casting|CastToMeatballVehicle_L (Variables|Default|GetCachedVehicleInterface)))
  (Class|MeatballVehicleL|DisableWeapon _asmeatball_vehicle_l (not _bweaponenabled))
  (Variables|/Motorboat/Vehicle/MeatballVehicle_L.MeatballVehicle_L_C|SetbEnableWeapons (Utilities|Struct|MakeScalableFloat (Math|Conversions|ToFloat(Boolean) _bweaponenabled)) _asmeatball_vehicle_l))

(fn HandleEditModeCollision)

(fn UpdateEffects
  ;[0.1085 0.9003 1] Play the post-spawn sound
  (Networking|CanExecuteCosmeticEvents
    (:True
      (if (Class|CreativeEnabledComponent|IsEnabled (Variables|Default|GetEnabledComponent))
        (Audio|PlaySoundatLocation (Variables|Cosmetics|GetTrap%20Ready%20Sound) (Transformation|GetActorLocation))))))

(fn ConstructionScript
  (|CallParentFunction))
