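; Creative device Blueprint for the motorboat spawner, extending a parent vehicle spawner with boat-specific options and behavior.
; Fuel and wrap overrides read and write user options, while config updates apply the weapon toggle to the cached Meatball vehicle.
; Initialization picks an opaque or translucent hologram material and async-loads the boat mesh for placement previews.
; Migration remaps the legacy fuel-use multiplier into the current option, and effects play a ready sound.
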(event EventInitializeDevice
  (bind _staticmeshcomponent (Variables|Debug|GetStaticMeshComponent))
  (|CallParentFunction)
  ;[1 1 1] Adjust the Material if we are in editor for the hologram, as Unreal isn't a fan of Selecting Translucent Materials(Or Just press "T")
  (if (and (Editor|InEditor) (Variables|Default|GetUseEditorPreview))
    (Rendering|Material|SetMaterial _staticmeshcomponent 0 "/Game/Creative/Devices/Generic/Materials/M_CP_Device_Hologram_Vehicles_Opaque.M_CP_Device_Hologram_Vehicles_Opaque")
    (else
      (Rendering|Material|SetMaterial _staticmeshcomponent 0 "/Game/Creative/Devices/Generic/Materials/M_CP_Device_Hologram_Vehicles.M_CP_Device_Hologram_Vehicles")
      (if (Building|IsPreviewBuildingActor)
        (bind _object (Utilities|AsyncLoadAsset (Variables|Default|GetBoatPreviewMesh))
          (:Completed
            (bind _asstatic_mesh (Utilities|Casting|CastToStaticMesh _object))
            (bind _creativepreviewonlymesh (Variables|Default|GetCreativePreviewOnlyMesh)
              (:then
                (Components|StaticMesh|SetStaticMesh _creativepreviewonlymesh _asstatic_mesh))
              (:else))))))))

(event EventUpdateVehicleConfigs
  (bind _vehicle_weapon_enabled (Variables|UserOptions|GetVehicle%20Weapon%20Enabled))
  ;[0.8079 1 0.6756 0.75] Handle Weapons Option on Meatball Vehicle by Setting Value
  (|CallParentFunction)
  (bind _asmeatball_vehicle_l (Utilities|Casting|CastToMeatballVehicle_L (Variables|Default|GetCachedVehicleInterface)))
  (Class|MeatballVehicleL|DisableWeapon _asmeatball_vehicle_l (not _vehicle_weapon_enabled))
  (Variables|/Motorboat/Vehicle/MeatballVehicle_L.MeatballVehicle_L_C|SetbEnableWeapons (Utilities|Struct|MakeScalableFloat (Math|Conversions|ToFloat(Boolean) _vehicle_weapon_enabled)) _asmeatball_vehicle_l))

(fn GetSupportsWraps
  (return (Variables|UserOptions|Basic|GetSupportsWraps)))

(fn GetConsumesFuel
  (return (Variables|UserOptions|GetFuel%20Consumption)))

(fn GetFuelUseMultiplier
  (return (Variables|UserOptions|GetFuel%20Use%20Multiplier)))

(fn GetMaxRandomStartingFuel
  (return (Variables|UserOptions|Advanced|GetMax%20Random%20Starting%20Fuel)))

(fn GetMinRandomStartingFuel
  (return (Variables|UserOptions|Advanced|GetMin%20Random%20Starting%20Fuel)))

(fn GetRandomStartingFuel
  (return (Variables|UserOptions|GetRandom%20Starting%20Fuel)))

(fn GetStartingFuelValue
  (return (Variables|UserOptions|GetStarting%20Fuel)))

(fn HandleEditorVisibility)

(fn HandleEditModeCollision)

(fn MigrateLegacyUserOptions
  (bind (_vehicle_weapon_enabled _fuelusemultiplier) (Class|GetClassDefaults "/CRD_VehicleSpawners/SetupAssets/MotorBoat/Device_MotorBoatSpawner_V2.Device_MotorBoatSpawner_V2_C"))
  (|CallParentFunction)
  (Utilities|FlowControl|Sequence
    (:then_0
      (Utilities|FlowControl|Sequence
        (:then_0
          (if (!= (Variables|Legacy|GetFuelUseMultiplier) _fuelusemultiplier)
            (Variables|UserOptions|SetFuel%20Use%20Multiplier (Math|Float|MapRangeClamped (Variables|Legacy|GetFuelUseMultiplier) 0.5 1.5 0.5 2.0))
            (Variables|Legacy|SetFuelUseMultiplier _fuelusemultiplier) ; LEGACY
          ))))
    (:then_1
      (if (LocalVariables|GetOptionMigrated)
        (OptionData|RefreshPropertyOverrides (Variables|Default|GetToyOptionsComponent))))))

(fn SetConsumeFuel
  (params (ConsumeFuel :type Boolean))
  (Variables|UserOptions|SetFuel%20Consumption ConsumeFuel))

(fn SetRandomStartingFuel
  (params (RandomStartingFuel :type Boolean))
  (Variables|UserOptions|SetRandom%20Starting%20Fuel RandomStartingFuel))

(fn SetStartingFuelValue
  (params (Starting%20Fuel%20Value :type Integer))
  (Variables|UserOptions|SetStarting%20Fuel Starting%20Fuel%20Value))

(fn SetSupportsWraps
  (params (Supports%20Wraps :type Boolean))
  (Variables|UserOptions|Basic|SetSupportsWraps Supports%20Wraps))

(fn SetMaxRandomStartingFuel
  (params (Max%20Fuel%20Percent :type Integer))
  (|CallParentFunction Max%20Fuel%20Percent)
  (Variables|UserOptions|Advanced|SetMax%20Random%20Starting%20Fuel Max%20Fuel%20Percent))

(fn SetMinRandomStartingFuel
  (params (Min%20Fuel%20Percent :type Integer))
  (|CallParentFunction Min%20Fuel%20Percent)
  (Variables|UserOptions|Advanced|SetMin%20Random%20Starting%20Fuel Min%20Fuel%20Percent))

(fn UpdateEffects
  ;[0.1085 0.9003 1] Play the post-spawn sound
  (Networking|CanExecuteCosmeticEvents
    (:True
      (if (and (Class|CreativeEnabledComponent|IsEnabled (Variables|Default|GetEnabledComponent)) (not (Editor|InEditor)))
        (Audio|PlaySoundatLocation (Variables|Cosmetics|GetTrap%20Ready%20Sound) (Transformation|GetActorLocation))))))

(fn ConstructionScript
  (|CallParentFunction))
